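# Niagara system for the LinedNotebook river prop, named as trash, that drifts green sprite debris along a flowing surface.
# Two distance-culled GPU emitters spawn continuously on a thin box and flow along -Y: longer-lived dark particles, plus short-lived, bright, low-alpha ones with dynamic material fade.
# Curl noise adds wander, zeroed Z force and velocity keep motion flat, and waveforms vary sprite size.
# SpawnCap kills and camera-relative attraction are disabled; mesh orientation settings remain despite sprite-only renderers.
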NiagaraSystem /Game/Environments/Helios/Props/LinedNotebook/LinedNotebook_River/Effects/NS_LinedNotebook_River_Trash
dependencies {
 usedDataInterfaces [&NiagaraDataInterfaceMeshRendererInfo]
 usedDynamicInputs [
  &MakeFloatFromVector
  &MakeVectorFromLinearColorRGB
  &self:ScratchDynamicInput
 ]
 usedModules [
  &SystemState
  &self:SystemSpawnScript.SystemScriptSource.SystemScriptGraph.NiagaraNodeAssignment_0.SetVariables_#1
  &self:SystemSpawnScript.SystemScriptSource.SystemScriptGraph.NiagaraNodeAssignment_1.SetVariables_#2
 ]
}
emitters {
 Leaves_HIGH {
  simTarget=GPUComputeSim
  emitterUpdateScript [
   &EmitterState(
    Enable Distance Culling=true
    MaxDistance=10000
    Module.bApplySpawnCountScaleByDistance=false
    Module.bUseMaxDistance=true
    Module.bUseMinDistance=false
    Scalability Mode=::NewEnumerator1
    Scale Spawn Count=true
    Spawn Count Scale By Distance=DataInterface() {propertyValues {Curve {keys [<at 0 0>,<at 0.2 1>,<at 0.75 1>,<at 1 0>]}}}
   )
   &SpawnRate(SpawnRate=500, Use Spawn Probability=false)
  ]
  particleSpawnScript [
   &InitializeParticle(
    Color=&MultiplyLinearColor(
     Linear Color=<rgba 0.0097 0.0417 0.0005 1>
     Scale Factor=&MakeVector4_FromFloat(Float=&RandomRangeFloat(Minimum=1, Maximum=1))
    )
    Lifetime Max=6
    Lifetime Min=5
    Lifetime Mode=::NewEnumerator1
    Mesh Renderer Array Visibility Mode=::NewEnumerator3
    Mesh Renderer Info=DataInterface() {propertyValues {MeshRenderer=&self:Leaves_HIGH_0.NiagaraMeshRendererProperties_1}}
    Mesh Scale Max=<0.03 0.03 0.01>
    Mesh Scale Min=<0.02 0.02 0.005>
    Mesh Scale Mode=::NewEnumerator2
    Sprite Size Mode=::NewEnumerator1
    Uniform Sprite Size Max=50
    UsePositionOffset=false
    Write Material Random=true
   )
   &ShapeLocation(
    Box Size=<300 800 5>
    MaskSpawnGroup=false
    Offset=<0 0 -0.5>
    Offset Mode=::NewEnumerator0
    Override Seed=false
    Shape Primitive=::NewEnumerator2
    Surface Only=true
    Surface Only Band Thickness=1
    UseOwnerScale=true
   )
   &AddVelocity(Velocity=&RandomRangeVector(Minimum=<0 -125 0>, Maximum=<0 -150 0>))
   &ApplyInitialForces()
   &InitialMeshOrientation(
    Mesh Axis To Orient=<0 1 0>
    Mesh Orientation Mode=::NewEnumerator2
    Rotation=&RandomRangeVector(Minimum=<0 0 0>, Maximum=<0 0 1>)
    Use Rotation Vector=true
   )
   /SetVariables(Particles.VelocityClamp:Vector3f=&Multiply_Vector(A=$Particles.Velocity, B=<1 1 0>)) {isSetParameters=true, module=SetVariables_#3}
   &KillParticles(Kill Particles=&SetBoolByIntComparison(A=$Engine.Emitter.TotalSpawnedParticles, B=$User.SpawnCap, Comparison Type=::NewEnumerator0)) {enabled=false}
  ]
  particleUpdateScript [
   &ParticleState()
   &ScaleSpriteSize(
    Uniform Curve Scale=&Waveform(
     Global Amplitude Scale=&RandomRangeFloat(Minimum=0.7, Maximum=1)
     "Waveform [1]"=::NewEnumerator0
     "[1] Phase Scale"=&RandomRangeFloat(Minimum=1, Maximum=2)
     "[1] Frequency"=&RandomRangeFloat(Minimum=1, Maximum=2)
     "[1] Amplitude Min/Max"=<0 1>
    )
    Uniform Curve Sprite Scale=DataInterface() {propertyValues {Curve {keys [{tangentMode=RCTM_User, time=0, value=0.5}, {tangentMode=RCTM_User, time=1, value=1}]}}}
   )
   &ScaleColor(
    Scale RGB=&VectorFromFloat(Value=&FloatFromCurve(FloatCurve=DataInterface() {propertyValues {Curve {keys [<at 0.8 1>,<at 1 0>]}}}, Scale Curve=10))
    ScaleA=true
    ScaleRGB=true
   )
   &MeshRotationForce(Rotation=&MakeVector(
     X=&Multiply_Float(A=1, B=&FloatFromCurve(FloatCurve=$User.FallCurve))
     Y=&Multiply_Float(A=1, B=&FloatFromCurve(FloatCurve=$User.FallCurve))
     Z=&Multiply_Float(A=1, B=&FloatFromCurve(FloatCurve=$User.FallCurve))
    )) {module=MeshRotationForce001}
   &CurlNoiseForce(
    Noise Frequency=&RandomRangeFloat(Minimum=2, Maximum=10)
    Noise Strength=&RandomRangeFloat(Minimum=25, Maximum=50)
    Pan Noise=false
   )
   &SolveForcesAndVelocity(
    Clamp Velocity=true
    Force:Vector3f=&Multiply_Vector(A=$Transient.PhysicsForce, B=<1 1 0>)
    Limit Acceleration=false
    Speed Limit=250
    Velocity:Vector3f=&Multiply_Vector(A=$Particles.Velocity, B=<1 1 0>)
   )
   &ScaleSpriteSizeBySpeed()
   /Niagara/Modules/Update/Forces/PointAttractionForce(
    Attraction Radius:NiagaraFloat=&Waveform(
     Waveform Count=::NewEnumerator1
     Global Amplitude Scale=250
     "Waveform [1]"=::NewEnumerator0
     "[1] Frequency"=5
     "[1] Amplitude Min/Max"=<0.4 1>
     "[2] Average"=false
     "[2] Input Position"=0
     "[2] Phase"=$Emitter.Age
     "[2] Phase Scale"=1
     "[2] Compound Frequency"=2
     "[2] Frequency"=2
     "[2] Amplitude Min/Max"=<0 1>
     "[2] Scale"=1
     "[2] Bias"=0
     "[2] Exponent"=1
     "[2] Pulse Duration"=0.1
     "[2] Clip Triangle Min/Max"=<0 0>
     "[2] Random Seed"=0
    )
    AttractionStrength:NiagaraFloat=&Multiply_Float(
     A=&FloatFromCurve(FloatCurve=DataInterface() {propertyValues {Curve {keys [<at 0 1>,<at 0.25 0>]}}}, Scale Curve=-250)
     B=1
    )
    AttractorPosition:Vector3f=&Add_Vector(
     A=$Emitter.CameraPosition
     B=&Add_Vector(
      A=&Add_Vector(
       A=&Multiply_VectorByFloat(Vector=<1 1 1>, Float=-130)
       B=&Multiply_VectorByFloat(Vector=$Emitter.CameraForward, Float=250)
      )
      B=&Multiply_VectorByFloat(Vector=$Emitter.CameraRight, Float=-50)
     )
    )
    Falloff Exponent=1
    Module.Kill Within Radius=false
    Module.Use Falloff=false
   ) {enabled=false, module=PointAttractionForce001}
  ]
  rendererClasses [&NiagaraSpriteRendererProperties]
  renderers [{rendererClass=&NiagaraSpriteRendererProperties}]
 }
 Leaves_HIGH001 {
  simTarget=GPUComputeSim
  emitterUpdateScript [
   &EmitterState(
    Enable Distance Culling=true
    MaxDistance=10000
    Module.bApplySpawnCountScaleByDistance=false
    Module.bUseMaxDistance=true
    Module.bUseMinDistance=false
    Scalability Mode=::NewEnumerator1
    Scale Spawn Count=true
    Spawn Count Scale By Distance=DataInterface() {propertyValues {Curve {keys [<at 0 0>,<at 0.2 1>,<at 0.75 1>,<at 1 0>]}}}
   )
   &SpawnRate(SpawnRate=1000, Use Spawn Probability=false)
  ]
  particleSpawnScript [
   &InitializeParticle(
    Color=&RandomRangeLinearColor(Minimum=<rgba 0.2078 1 0 1>, Maximum=<rgba 0.3658 1 0 1>)
    Lifetime Max=1.25
    Lifetime Min=0.5
    Lifetime Mode=::NewEnumerator1
    Sprite Rotation Mode=::NewEnumerator2
    Sprite Size Mode=::NewEnumerator1
    Uniform Sprite Size Max=80
    Uniform Sprite Size Min=50
    UsePositionOffset=false
    Write Material Random=true
   )
   &ShapeLocation(
    Box Size=<300 800 5>
    MaskSpawnGroup=false
    Offset=<0 0 -15>
    Offset Mode=::NewEnumerator0
    Override Seed=false
    Shape Primitive=::NewEnumerator2
    Surface Only=true
    Surface Only Band Thickness=1
    UseOwnerScale=true
   )
   &AddVelocity(Velocity=&RandomRangeVector(Minimum=<0 -200 0>, Maximum=<0 -250 0>))
   &ApplyInitialForces()
   &InitialMeshOrientation(
    Mesh Axis To Orient=<0 1 0>
    Mesh Orientation Mode=::NewEnumerator2
    Rotation=&RandomRangeVector(Minimum=<0 0 0>, Maximum=<0 0 1>)
    Use Rotation Vector=true
   )
   /SetVariables(Particles.VelocityClamp:Vector3f=&Multiply_Vector(A=$Particles.Velocity, B=<1 1 0>)) {isSetParameters=true, module=SetVariables_#3}
   &KillParticles(Kill Particles=&SetBoolByIntComparison(A=$Engine.Emitter.TotalSpawnedParticles, B=$User.SpawnCap, Comparison Type=::NewEnumerator0)) {enabled=false}
  ]
  particleUpdateScript [
   &ParticleState()
   &SpriteRotationRate()
   &ScaleSpriteSize(
    Uniform Curve Scale=&Waveform(
     Global Amplitude Scale=&RandomRangeFloat(Minimum=0.7, Maximum=1)
     "Waveform [1]"=::NewEnumerator0
     "[1] Phase Scale"=&RandomRangeFloat(Minimum=0.25, Maximum=1)
     "[1] Frequency"=&RandomRangeFloat(Minimum=1, Maximum=2)
     "[1] Amplitude Min/Max"=<0 1>
    )
    Uniform Curve Sprite Scale=DataInterface() {propertyValues {Curve {keys [{tangentMode=RCTM_User, time=0, value=0.5}, {tangentMode=RCTM_User, time=1, value=1}]}}}
   )
   &ScaleColor(
    Scale Alpha=0.25
    Scale RGB=&VectorFromFloat(Value=&FloatFromCurve(FloatCurve=DataInterface() {propertyValues {Curve {keys [<at 0.8 1>,<at 1 0>]}}}, Scale Curve=5))
    ScaleA=true
    ScaleRGB=true
   )
   &CurlNoiseForce(
    Noise Frequency=&RandomRangeFloat(Minimum=2, Maximum=10)
    Noise Strength=&RandomRangeFloat(Minimum=25, Maximum=50)
    Pan Noise=false
   )
   &SolveForcesAndVelocity(
    Clamp Velocity=true
    Force:Vector3f=&Multiply_Vector(A=$Transient.PhysicsForce, B=<1 1 0>)
    Limit Acceleration=false
    Speed Limit=250
    Velocity:Vector3f=&Multiply_Vector(A=$Particles.Velocity, B=<1 1 0>)
   )
   &ScaleSpriteSizeBySpeed()
   &PointAttractionForce(
    Attraction Radius:NiagaraFloat=&Waveform(
     Waveform Count=::NewEnumerator1
     Global Amplitude Scale=250
     "Waveform [1]"=::NewEnumerator0
     "[1] Frequency"=5
     "[1] Amplitude Min/Max"=<0.4 1>
     "[2] Average"=false
     "[2] Input Position"=0
     "[2] Phase"=$Emitter.Age
     "[2] Phase Scale"=1
     "[2] Compound Frequency"=2
     "[2] Frequency"=2
     "[2] Amplitude Min/Max"=<0 1>
     "[2] Scale"=1
     "[2] Bias"=0
     "[2] Exponent"=1
     "[2] Pulse Duration"=0.1
     "[2] Clip Triangle Min/Max"=<0 0>
     "[2] Random Seed"=0
    )
    AttractionStrength:NiagaraFloat=&Multiply_Float(
     A=&FloatFromCurve(FloatCurve=DataInterface() {propertyValues {Curve {keys [<at 0 1>,<at 0.25 0>]}}}, Scale Curve=-250)
     B=1
    )
    AttractorPosition:Vector3f=&Add_Vector(
     A=$Emitter.CameraPosition
     B=&Add_Vector(
      A=&Add_Vector(
       A=&Multiply_VectorByFloat(Vector=<1 1 1>, Float=-130)
       B=&Multiply_VectorByFloat(Vector=$Emitter.CameraForward, Float=250)
      )
      B=&Multiply_VectorByFloat(Vector=$Emitter.CameraRight, Float=-50)
     )
    )
    Falloff Exponent=1
    Module.Kill Within Radius=false
    Module.Use Falloff=false
   ) {enabled=false, module=PointAttractionForce001}
   &DynamicMaterialParameters(
    Index 0 Param 1=&FloatFromCurve(FloatCurve=DataInterface() {propertyValues {Curve {keys [<at 0 10>,<at 1 0>]}}})
    Module.Param0WriteEnabled=true
    Module.Param1WriteEnabled=false
    Module.Param2WriteEnabled=false
    Module.Param3WriteEnabled=false
   )
   &CameraOffset(Camera Offset Amount=5)
  ]
  rendererClasses [&NiagaraSpriteRendererProperties]
  renderers [
   {rendererClass=&NiagaraSpriteRendererProperties}
   {rendererClass=&NiagaraSpriteRendererProperties, rendererIndex=1}
  ]
 }
}
summary {
 systemName=NS_LinedNotebook_River_Trash
 emitters [
  {emitterName=Leaves_HIGH, simTarget=GPUComputeSim, rendererClasses [&NiagaraSpriteRendererProperties]}
  {emitterName=Leaves_HIGH001, simTarget=GPUComputeSim, rendererClasses [&NiagaraSpriteRendererProperties]}
 ]
 userVariables [
  User.CurlCurve {
   type=NiagaraDataInterfaceCurve
   description=""
   defaultValue {
    struct=NiagaraExt_VariableValue_DataInterface
    value {
     dataInterface=&self:NiagaraDataInterfaceCurve_1
     dataInterfaceClass=&NiagaraDataInterfaceCurve
    }
   }
  }
  User.FallCurve {
   type=NiagaraDataInterfaceCurve
   description=""
   defaultValue {
    struct=NiagaraExt_VariableValue_DataInterface
    value {
     dataInterface=&self:NiagaraDataInterfaceCurve_0
     dataInterfaceClass=&NiagaraDataInterfaceCurve
    }
   }
  }
  User.RadiusSpread {type=NiagaraFloat, defaultValue=1000, description=""}
  User.SpawnCap {type=NiagaraInt32, defaultValue=600}
  User.SpawnRate {type=NiagaraFloat, defaultValue=15, description=""}
 ]
}
