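# Library-exposed material function that computes one Gerstner wave for the LinedNotebook river prop, from direction, wavelength, amplitude and steepness inputs.
# Wavelength gives the wavenumber, and a gravity parameter sets speed through deep-water dispersion, forming a phase from offset-free world XY and time.
# Sine and cosine of that phase drive a WPO output with vertical displacement and steepness-scaled horizontal pinching, plus an analytic, unnormalized Normals output.
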MaterialFunction /Game/Environments/Helios/Props/LinedNotebook/LinedNotebook_River/Materials/GerstnerWave_Fn
graph {
 properties {bExposeToLibrary=true}
 nodes [
  n1 = WorldPosition() {worldPositionShaderOffset=WPT_ExcludeAllShaderOffsets}
  n2 = FunctionInput() {bUsePreviewValueAsDefault=true, inputName=Wave Direction, previewValue=<1 0 0 1>}
  n3 = Normalize(VectorInput=ComponentMask(None=@n2) {g=true, r=true})
  n4 = FunctionInput() {bUsePreviewValueAsDefault=true, inputName=WaveLength, inputType=FunctionInput_Scalar, previewValue=<2048 0 0 1>}
  n5 = Divide(A=MaterialFunctionCall(Multiplier=Constant() {r=2}) {materialFunction=&Pi}.Result, B=@n4) {desc=k}
  n6 = Multiply(A=@n5, B=ScalarParameter() {defaultValue=980, parameterName=Gravitational Constant})
  n7 = Subtract(
   A=DotProduct(A=ComponentMask(None=@n1.XYZ) {g=true, r=true}, B=Multiply(A=@n3, B=@n5) {desc=K})
   B=Multiply(A=SquareRoot(None=@n6) {desc=w}, B=Time())
  )
  n8 = Cosine(None=@n7) {period=6.28}
  n9 = MaterialFunctionCall(Z=@n8) {materialFunction=&MakeFloat3}
  n10 = FunctionInput() {bUsePreviewValueAsDefault=true, inputName=Amplitude, inputType=FunctionInput_Scalar, previewValue=<256 0 0 1>}
  n11 = FunctionInput() {bUsePreviewValueAsDefault=true, inputName=Steepness, inputType=FunctionInput_Scalar, previewValue=<1 0 0 1>}
  n12 = Constant() {r=2}
  n13 = Multiply(
   A=@n10
   B=Divide(A=MaterialFunctionCall(Multiplier=@n12) {materialFunction=&Pi}.Result, B=@n4) {desc=wk}
  ) {desc=wKA}
  n14 = Divide(A=@n11, B=@n13) {desc=Q}
  n15 = Sine(None=@n7) {period=6.28}
  n16 = Multiply(A=Multiply(A=Multiply(A=@n14, B=@n10), B=@n15), B=Normalize(VectorInput=ComponentMask(None=@n2) {g=true, r=true}))
  n17 = Constant()
  FunctionOutput(None=Subtract(A=Multiply(A=@n9.Result, B=@n10), B=AppendVector(A=@n16, B=@n17)) {desc=WPO}) {outputName=WPO}
  n18 = Multiply(A=@n13, B=@n15)
  n19 = Normalize(VectorInput=ComponentMask(None=@n2) {g=true, r=true})
  n20 = Multiply(A=Multiply(A=@n13, B=@n8), B=@n14)
  FunctionOutput(None=AppendVector(A=Multiply(A=@n18, B=@n19), B=OneMinus(None=@n20)) {desc=Normal}) {outputName=Normals}
 ]
}
