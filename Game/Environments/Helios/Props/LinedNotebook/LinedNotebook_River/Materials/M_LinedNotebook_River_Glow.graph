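# Default Lit glow material for the river in the Helios LinedNotebook prop.
# A tinted, tiled caustics texture is multiplied by a sharpened Fresnel term, whose normal comes from a separately tiled foam normal map.
# A GerstnerWave function computes world position offset from wave parameters.
# One UsePanning? static switch toggles both texture scrolls and selects between wave offset and a constant.
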Material /Game/Environments/Helios/Props/LinedNotebook/LinedNotebook_River/Materials/M_LinedNotebook_River_Glow
graph {
 properties {pixelDepthOffsetMode=PDOM_Legacy, usedShadingModels=Default Lit}
 nodes [
  n1 = Multiply(
   A=TextureCoordinate() {uTiling=5}
   B=ComponentMask(None=VectorParameter() {parameterName=Til, defaultValue=<rgba 0.1 0.02 0 1>}.RGB) {g=true, r=true}
  )
  n2 = StaticSwitchParameter(True=Panner(Coordinate=@n1, Speed=VectorParameter() {parameterName=Speed, defaultValue=<rgba 0.1 0.02 0 1>}.RGB), False=@n1) {defaultValue=true, parameterName=UsePanning?}
  n3 = ScalarParameter() {defaultValue=2, parameterName=ExponentIn}
  n4 = Multiply(A=TextureCoordinate(), B=ComponentMask(None=VectorParameter() {parameterName=Til2, defaultValue=<rgba 1 1 0 1>}.RGB) {g=true, r=true})
  n5 = VectorParameter() {parameterName=Speed2, defaultValue=<rgba 0.1 0.02 0 1>}
  n6 = TextureSample(UVs=StaticSwitchParameter(True=Panner(Coordinate=@n4, Speed=@n5.RGB), False=@n4) {defaultValue=true, parameterName=UsePanning?}) {samplerType=SAMPLERTYPE_LinearColor, texture=/FortWater/Textures/T_foam_Normal2}
  Multiply(
   A=Multiply(
    A=VectorParameter() {parameterName=Color, defaultValue=<rgba 0 0 0 1>}.RGB
    B=TextureSampleParameter2D(UVs=@n2) {parameterName=Param, texture=/Game/Global/Textures/Default/Caustics/T_Caustics_Softer}.G
   )
   B=Power(Base=Fresnel(ExponentIn=@n3, Normal=@n6.RGB)) {constExponent=4}
  )
  ScalarParameter() {parameterName=Opacity}
  n7 = MaterialFunctionCall!(
   WaveLength=ScalarParameter() {defaultValue=2048, parameterName="WaveLength (S)"}
   Amplitude=ScalarParameter() {defaultValue=256, parameterName="Amplitude (S)"}
   Steepness=ScalarParameter() {defaultValue=1, parameterName="Steepness (S)"}
   Wave Direction=VectorParameter() {parameterName="Wave Direction (V3)", defaultValue=<rgba 0 0 1 0>}.RGB
  ) {materialFunction=&/GerstnerWave_Fn}
  StaticSwitchParameter(True=@n7.WPO, False=Constant()) {defaultValue=true, parameterName=UsePanning?}
  Transform(None=@n7.WPO) {transformSourceType=TRANSFORMSOURCE_Local}
 ]
}
